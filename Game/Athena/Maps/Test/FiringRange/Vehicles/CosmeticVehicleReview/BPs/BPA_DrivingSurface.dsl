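; Driving surface actor for the cosmetic vehicle review test map, showing a selectable material on a plane.
; ValidateMaterial applies the key at CurrentMaterialIndex in the Materials map to the plane, resolving it from a soft reference.
; It runs from the construction script and from SetMaterialIndex, which updates the index.
; A TestPhysMat debug event traces down from the camera and prints the hit physical surface type, or FAILED.
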(event EventBeginPlay)

(event Collision|EventActorBeginOverlap
  (params (OtherActor :type Actor%20Object%20Reference)))

(event EventTick
  (params (DeltaSeconds :type Float%20%28single-precision%29)))

(event Custom|SetMaterialIndex
  (params (CurrentMaterialIndex :type Integer))
  (Variables|Default|SetCurrentMaterialIndex CurrentMaterialIndex)
  (CallFunction|ValidateMaterial))

(event Custom|TestPhysMat
  (bind (_outhit _returnvalue) (Collision|LineTraceForObjects :Start (Camera|GetCameraLocation (Game|GetPlayerCameraManager 0)) :End (- (Camera|GetCameraLocation (Game|GetPlayerCameraManager 0)) "0.000000,0.000000,9999.000000") :ObjectTypes (Utilities|Array|MakeArray "ObjectTypeQuery1") :bTraceComplex true :DrawDebugType "ForDuration" :bIgnoreSelf true :TraceColor "(R=1.000000,G=0.000000,B=0.000000,A=1.000000)" :TraceHitColor "(R=0.000000,G=1.000000,B=0.000000,A=1.000000)" :DrawTime 5.0))
  (if _returnvalue
    (Development|PrintString (Utilities|String|EnumtoString (Variables|/Script/PhysicsCore.PhysicalMaterial|GetSurfaceType (.PhysMat (Collision|BreakHitResult _outhit)))))
    (else
      (Development|PrintString "FAILED" true true "(R=1.000000,G=0.068537,B=0.000000,A=1.000000)"))))

(fn ConstructionScript
  (CallFunction|ValidateMaterial))

(fn ValidateMaterial
  (bind _keys (Utilities|Map|Keys (Variables|Default|GetMaterials)))
  (Rendering|Material|SetMaterial (Variables|Default|GetPlane) 0 (Utilities|ResolveSoftReference (Utilities|Array|Get(acopy) _keys (Variables|Default|GetCurrentMaterialIndex)))))
